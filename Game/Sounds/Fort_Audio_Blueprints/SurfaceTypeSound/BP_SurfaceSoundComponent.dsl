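; Surface sound component that plays an attached sound whose content changes with the physical surface type.
; BeginPlay spawns the audio component on a configured attach component, and EndPlay fades it out.
; On Surface Type Changed, if the audio component is valid, it looks up the surface name in a data table, falling back to an empty entry.
; It pushes a random sound, volume and pitch as "Surfaces" audio parameters, then fires the OnSurfaceChanged trigger.
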(event EventBeginPlay
  (bind _returnvalue (Audio|SpawnSoundAttached (Variables|Default|GetSound) (Variables|Default|GetAttach%20to%20Component)))
  (Variables|Default|SetSurface%20Audio%20Component _returnvalue))

(event Custom|Surface%20Type%20Changed
  (params (Surface%20Type :type EPhysicalSurface%20Enum :default "SurfaceType_Default"))
  (bind (_surface_sounds_data_surfacesound_8_4fe7bd224137f616108d6d9a28013055 _surface_sounds_data_volumemultiplier_10_370561ff448c62ed90c9a0abaaad3e40 _surface_sounds_data_pitchshift_9_2e667f4b45c83f3fbb015e9c0978b909) (Utilities|Struct|BreakSurfaceSoundStruct (Variables|GetSurface%20Sounds%20Data_SurfaceSound_8_4FE7BD224137F616108D6D9A28013055)))
  (Variables|Default|GetSurface%20Audio%20Component
    (:then
      (bind _returnvalue (Utilities|GetDataTableRow (Variables|Default|GetSurface%20Sound%20Data%20Table) (Fortnite|Utility|GetPhysicalSurfaceName Surface%20Type))
        (:then
          (Variables|Default|SetSurface%20Sounds%20Data _returnvalue))
        (:RowNotFound
          (Variables|Default|SetSurface%20Sounds%20Data (Utilities|Struct|MakeSurfaceSoundStruct (Utilities|Array|MakeArray 0)))))
      (Audio|Parameter|SetObjectParameter (Variables|Default|GetSurface%20Audio%20Component) "Surfaces.Sound" (select (Utilities|Array|IsNotEmpty _surface_sounds_data_surfacesound_8_4fe7bd224137f616108d6d9a28013055) (Utilities|Array|RandomArrayItem _surface_sounds_data_surfacesound_8_4fe7bd224137f616108d6d9a28013055) 0))
      (Audio|Parameter|SetFloatParameter (Variables|Default|GetSurface%20Audio%20Component) "Surfaces.Volume" _surface_sounds_data_volumemultiplier_10_370561ff448c62ed90c9a0abaaad3e40 :!class /Script/AudioExtensions.AudioParameterControllerInterface)
      (Audio|Parameter|SetFloatParameter (Variables|Default|GetSurface%20Audio%20Component) "Surfaces.Pitch" _surface_sounds_data_pitchshift_9_2e667f4b45c83f3fbb015e9c0978b909 :!class /Script/AudioExtensions.AudioParameterControllerInterface)
      (Audio|Parameter|ExecuteTriggerParameter (Variables|Default|GetSurface%20Audio%20Component) "Surfaces.OnSurfaceChanged"))
    (:else)))

(event EventEndPlay
  (params (EndPlayReason :type EEndPlayReason%20Enum :default "Destroyed"))
  (Audio|Components|Audio|FadeOut (Variables|Default|GetSurface%20Audio%20Component) 0.3 0.0 "Linear" :!class /Script/Engine.AudioComponent))
