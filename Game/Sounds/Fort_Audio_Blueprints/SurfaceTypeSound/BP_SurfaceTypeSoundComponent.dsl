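; Blueprint component that plays a surface-specific sound at a given location, chosen from a sound map keyed by physical surface.
; SetSurfaceType derives the surface from a BuildingSMActor's resource type, otherwise stores the passed-in surface, ignoring invalid actors.
; PlaySound requires a map match and a set surface, then saves the location and calls ActuallyPlaySound.
; A leftover Test Delay event only prints a debug string.
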(event Custom|Test%20Delay
  (Development|PrintString "Hello"))

(fn ActuallyPlaySound
  (Audio|PlaySoundatLocation (Utilities|Map|Find (Variables|Default|GetSoundMap) (Variables|Default|GetSurfaceType)) (Variables|Default|GetLocation)))

(fn PlaySound
  (params (Location :type Vector))
  (bind (_value _returnvalue) (Utilities|Map|Find (Variables|Default|GetSoundMap) (Variables|Default|GetSurfaceType)))
  (if (and _returnvalue (Variables|Default|GetIsSurfaceTypeSet))
    (Variables|Default|SetLocation Location)
    (CallFunction|ActuallyPlaySound)))

(fn SetSurfaceType
  (params (SurfaceType :type EPhysicalSurface%20Enum :default "SurfaceType_Default") (Actor :type Actor%20Object%20Reference))
  (Utilities|IsValid Actor
    (:Is%20Valid
      (bind _asbuilding_static_mesh (Utilities|Casting|CastToBuildingSMActor Actor)
        (:then
          (Variables|Default|SetSurfaceType (Utilities|Select :Index (Variables|/Script/FortniteGame.BuildingSMActor|GetResourceType _asbuilding_static_mesh) :Wood "SurfaceType1" :Stone "SurfaceType2" :Metal "SurfaceType3" :Permanite "SurfaceType_Default" :GoldCurrency "SurfaceType_Default" :Ingredient "SurfaceType_Default" :None "SurfaceType_Default")))
        (:CastFailed
          (Variables|Default|SetSurfaceType SurfaceType)))
      (Variables|Default|SetIsSurfaceTypeSet true))
    (:Is%20Not%20Valid
      _)))
